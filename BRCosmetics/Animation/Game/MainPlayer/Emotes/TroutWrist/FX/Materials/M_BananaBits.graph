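# M_BananaBits is a masked, two-sided, default-lit material for Niagara mesh particles, named for banana bits in the TroutWrist emote effects.
# It holds samples of the emote prop's diffuse and normal textures, plus a constant left at its default.
# Automatic usage setting in the editor is off, with Niagara mesh particle usage enabled explicitly.
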Material /BRCosmetics/Animation/Game/MainPlayer/Emotes/TroutWrist/FX/Materials/M_BananaBits
graph {
 properties {
  bAutomaticallySetUsageInEditor=false
  bUsedWithNiagaraMeshParticles=true
  blendMode=BLEND_Masked
  twoSided=true
  usedShadingModels=Default Lit
 }
 nodes [
  Constant()
  TextureSample() {texture=/BRCosmetics/Animation/Game/MainPlayer/Emotes/TroutWrist/FX/Textures/T_TroutWrist_Emote_prop_D}
  TextureSample() {samplerType=SAMPLERTYPE_Normal, texture=/BRCosmetics/Animation/Game/MainPlayer/Emotes/TroutWrist/FX/Textures/T_TroutWrist_Emote_prop_N}
 ]
}
