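# Masked, two-sided lit material for the banana effect in the TroutWrist emote, built for Niagara mesh particles and ribbons.
# Named reroutes build tan-brown color from smoke and panning crack textures, and normals from flattened sphere and dirt-path maps.
# Opacity is a UV sphere mask with radius driven by world-space noise and the Dither dynamic parameter.
# World position offset uses time-scrolled noise along Z, shaped by a center ramp and UV gradient.
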Material /BRCosmetics/Animation/Game/MainPlayer/Emotes/TroutWrist/FX/Materials/M_Emote_Banana
graph {
 properties {
  bAutomaticallySetUsageInEditor=false
  bUsedWithNiagaraMeshParticles=true
  bUsedWithNiagaraRibbons=true
  blendMode=BLEND_Masked
  opacityMaskClipValue=0.5
  pixelDepthOffsetMode=PDOM_Legacy
  twoSided=true
  usedShadingModels=Default Lit
 }
 nodes [
  n1 = TextureCoordinate() {coordinateIndex=1}
  n2 = Add(
   A=Multiply(A=Subtract(A=@n1) {constB=0.5}, B=ScalarParameter() {defaultValue=0.5, parameterName=B_3}) {constB=0.5}
   B=AppendVector(A=ScalarParameter() {parameterName=X_Offset}, B=ScalarParameter() {defaultValue=0.5, parameterName=Y_Offset})
  )
  n3 = Multiply(A=TextureSample(UVs=Panner(Coordinate=@n2) {speedX=-1}) {texture=/BRCosmetics/Athena/Effects/Textures/T_InkySmokeTrail}.A)
  n4 = NamedRerouteUsage()
  Add(A=@n3, B=Multiply(A=NamedRerouteUsage(), B=@n4))
  Multiply(A=@n3, B=@n4)
  Multiply(
   A=ParticleColor().RGB
   B=TextureSample() {texture=/BRCosmetics/Animation/Game/MainPlayer/Emotes/TroutWrist/FX/Textures/T_TroutWrist_Emote_prop_D}.RGB
  )
  Multiply(A=WorldPosition().XY)
  n5 = Constant3Vector() {constant=<rgba 0.7305 0.5149 0.2346 1>}
  n6 = Constant3Vector() {constant=<rgba 0.4948 0.3488 0.1589 1>}
  n7 = TextureSample() {texture=/Game/Global/Textures/Default/Tileable_Patterns/T_Blowing_Smoke_Pattern2_Tile_DISP}
  n8 = Panner() {speedX=-2}
  NamedRerouteDeclaration(None=LinearInterpolate(
    A=LinearInterpolate(A=@n5, B=@n6, Alpha=@n7.R)
    B=Constant3Vector() {constant=<rgba 0.5972 0.3095 0.0908 1>}
    Alpha=TextureSample(UVs=@n8) {samplerType=SAMPLERTYPE_LinearColor, texture=/Game/Global/Textures/Default/Masks_and_Gradients/T_Cracks}.R
   )) {name=Col, nodeColor=<rgba 0 0.3437 1 1>}
  n9 = TextureCoordinate()
  n10 = DynamicParameter() {defaultValue=<rgba 60 0.005 0.8 1>, paramNames [DisplaceAmt, WorldPosScale, Dither, Param4]}
  n11 = Noise!(
   World Position=Multiply(A=WorldPosition().XYZ, B=ScalarParameter() {defaultValue=0.02, parameterName=OpacityWSMult})
   FilterWidth=ScalarParameter() {parameterName=FilterWidth}
  ) {levelScale=6, outputMin=0}
  NamedRerouteDeclaration(None=SphereMask(
    A=ComponentMask(None=@n9) {g=true}
    B=ScalarParameter() {defaultValue=0.5, parameterName="Pos_)"}
    Radius=Add(A=@n10.Dither, B=@n11)
    Hardness=ScalarParameter() {defaultValue=0.02, parameterName=Param_2}
   )) {name=Opacity, nodeColor=<rgba 1 0 0.9844 1>}
  n12 = TextureCoordinate() {coordinateIndex=1}
  n13 = NamedRerouteDeclaration(None=SphereMask(
    A=ComponentMask(None=@n12) {g=true}
    B=ScalarParameter() {defaultValue=0.5, parameterName=CenterU}
    Radius=ScalarParameter() {defaultValue=0.3522, parameterName=RadiusU}
    Hardness=ScalarParameter() {defaultValue=0.1, parameterName=HardnessU}
   )) {name=CenterRampMiddle, nodeColor=<rgba 1 0 0.4922 1>}
  n14 = Subtract(A=ObjectPositionWS(), B=WorldPosition().XYZ)
  n15 = Multiply(A=Time(), B=OneMinus(None=Constant3Vector() {constant=<rgba 0 30 0 0>}))
  n16 = Noise(World Position=Multiply(A=Add(A=@n14, B=@n15), B=@n10.WorldPosScale), FilterWidth=ScalarParameter() {parameterName=FilterWidth}) {levelScale=6, outputMin=0}
  n17 = Multiply(A=Constant3Vector() {constant=<rgba 0 0 1 0>}, B=ScalarParameter() {defaultValue=5, parameterName=WPODis})
  n18 = OneMinus(None=StaticComponentMaskParameter(None=TextureCoordinate()) {defaultR=true})
  NamedRerouteDeclaration(None=Multiply(A=Multiply(A=@n13, B=@n16), B=Multiply(A=@n17, B=@n18))) {name=WPO, nodeColor=<rgba 0 1 0.6953 1>}
  n19 = TextureCoordinate() {coordinateIndex=1}
  n20 = AppendVector(A=ScalarParameter() {defaultValue=1, parameterName=N2_UMult}, B=ScalarParameter() {defaultValue=1, parameterName=N2_VMult})
  n21 = TextureCoordinate()
  n22 = AppendVector(A=ScalarParameter() {defaultValue=0.76, parameterName=N_UMult}, B=ScalarParameter() {defaultValue=0.2, parameterName=N_VMult})
  n23 = MaterialFunctionCall(
   BaseNormal=TextureSample(UVs=Multiply(A=@n19, B=@n20)) {samplerType=SAMPLERTYPE_Normal, texture=/BRCosmetics/Animation/Game/MainPlayer/Emotes/TroutWrist/FX/Textures/T_FX_SphereNormals}.RGB
   AdditionalNormal=TextureSample(UVs=Multiply(A=@n21, B=@n22)) {samplerType=SAMPLERTYPE_LinearColor, texture=/Game/Environments/Asteria/Roads/Knights/Cobblestone/Textures/DirtPath_normal_SMRH}.RGB
  ) {materialFunction=&BlendAngleCorrectedNormals}
  n24 = ScalarParameter() {defaultValue=0.6, parameterName="Flatness (S)"}
  NamedRerouteDeclaration(None=NamedRerouteDeclaration(None=MaterialFunctionCall(Normal=@n23.Result, Flatness=@n24) {materialFunction=&FlattenNormal}.Result) {name=SphericalN, nodeColor=<rgba 0.2422 1 0 1>}) {name=Normals, nodeColor=<rgba 1 0.5859 0 1>}
  n25 = TextureCoordinate() {coordinateIndex=1}
  n26 = ScalarParameter() {defaultValue=0.6, parameterName=RadiusMaskMinU}
  n27 = ScalarParameter() {defaultValue=0.3, parameterName=RadiusMaskMaxU}
  n28 = Noise(World Position=Multiply(A=WorldPosition().XYZ, B=ScalarParameter() {defaultValue=0.03, parameterName=PScale})) {levels=2, outputMin=0}
  NamedRerouteDeclaration(None=SphereMask(
    A=ComponentMask(None=@n25) {g=true}
    B=ScalarParameter() {defaultValue=0.5, parameterName=CenterMaskU}
    Radius=LinearInterpolate(A=@n26, B=@n27, Alpha=@n28)
    Hardness=ScalarParameter() {defaultValue=0.5, parameterName=HardnessMaskU}
   )) {name=CenterRampMask, nodeColor=<rgba 1 0 0.4922 1>}
  NamedRerouteDeclaration(None=Multiply(A=NamedRerouteUsage(), B=VectorParameter() {parameterName=Subsurface Color, defaultValue=<rgba 0.2232 0.0865 0.0497 1>}.RGB)) {name=subSurfaceColor, nodeColor=<rgba 0 0.8125 1 1>}
  n29 = TextureCoordinate() {coordinateIndex=1}
  n30 = AppendVector(A=Constant() {r=0.3}, B=Constant())
  NamedRerouteDeclaration(None=SphereMask(
    A=Add(A=@n29, B=@n30) {constB=0}
    B=ScalarParameter() {defaultValue=0.5, parameterName=CenterMaskU}
    Radius=Constant() {r=0.7}
    Hardness=ScalarParameter() {defaultValue=0.5, parameterName=HardnessMaskU}
   )) {name=MaskArea, nodeColor=<rgba 0.9688 1 0 1>}
  NamedRerouteUsage()
  NamedRerouteUsage()
  NamedRerouteUsage()
  NamedRerouteUsage()
  ScalarParameter() {parameterName=Spec}
  ScalarParameter() {defaultValue=0.6, parameterName=Roughness}
  ScalarParameter() {defaultValue=0.005, parameterName=B_1}
  TextureSample(UVs=@n8) {samplerType=SAMPLERTYPE_LinearGrayscale, texture=/BRCosmetics/Weapons/FORT_Melee/Pickaxe_Cube_Ninja_Male/FX/T_FluidFlameTile}
 ]
}
