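# Material function in the TroutWrist emote FX folder that converts a frame number into UVs for one cell of a SubUV texture grid.
# It strips the fraction from Frame, derives column and row from the horizontal image count, then offsets the input UVs and scales them by 1/SubImages.
# Despite the description mentioning blending, only this frame's UVs are output; the next-frame UVs and the lerp fraction have no output.
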MaterialFunction /BRCosmetics/Animation/Game/MainPlayer/Emotes/TroutWrist/FX/Materials/SubUV_Function
graph {
 properties {description=Allows SubUV blending of multiple texture frames in a single texture, libraryCategoriesText [Misc,Texturing]}
 nodes [
  n1 = FunctionInput() {inputName=Frame, inputType=FunctionInput_Scalar, sortPriority=-10}
  n2 = Subtract(A=@n1, B=Frac(None=@n1) {desc=Frame Lerp}) {desc=IndexA}
  n3 = FunctionInput() {inputName=SubImages, inputType=FunctionInput_Vector2, sortPriority=-20, previewValue=<2 2 0 0>}
  n4 = ComponentMask(None=@n3) {desc="## Horz Images", r=true}
  n5 = Divide(A=Constant2Vector() {g=1, r=1}, B=@n3) {desc=Scale}
  n6 = ComponentMask(None=@n5) {desc="1 / ## Horz Images", r=true}
  n7 = AppendVector(A=Fmod(A=@n2, B=@n4) {desc=IndexA_Horz}, B=Floor(None=Multiply(A=@n2, B=@n6)) {desc=IndexA_Vert}) {desc=IndexA_Combined}
  n8 = FunctionInput(Preview=TextureCoordinate()) {bUsePreviewValueAsDefault=true, inputName=UVs, inputType=FunctionInput_Vector2, sortPriority=-30, previewValue=<2 2 0 0>}
  FunctionOutput(None=Multiply(A=Add(A=@n7, B=@n8), B=@n5) {desc=UV_A}) {outputName=UVs, sortPriority=3}
  n9 = Add(B=@n2) {constA=1, desc=IndexB}
  n10 = Fmod(A=@n9, B=@n4) {desc=IndexB_Horz}
  n11 = Floor(None=Multiply(A=@n9, B=@n6)) {desc=IndexB_Vert}
  Multiply(A=Add(A=@n8, B=AppendVector(A=@n10, B=@n11) {desc=IndexB_Combined}), B=@n5) {desc=UV_B}
 ]
}
