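# Translucent unlit deferred decal material in the TroutWrist emote FX, named for a trail and flagged for Niagara sprites and ribbons.
# It holds several separate branches: flowmap-warped splash erosion, SubUV liquid, panned smoke and blended, flattened sphere normals.
# Sphere masks, noise and dynamic parameters shape radius and breakup, while decal lifetime or particle alpha drives fades.
# No material output connections appear, so which branches are live is unclear.
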Material /BRCosmetics/Animation/Game/MainPlayer/Emotes/TroutWrist/FX/Materials/M_TroutTrail_Decal
graph {
 properties {
  bEnableResponsiveAA=true
  bOutputTranslucentVelocity=true
  bUsedWithNiagaraRibbons=true
  bUsedWithNiagaraSprites=true
  blendMode=BLEND_Translucent
  materialDomain=MD_DeferredDecal
  pixelDepthOffsetMode=PDOM_Legacy
  shadingModel=MSM_Unlit
  twoSided=true
  usedShadingModels=Unlit
 }
 nodes [
  Constant()
  LinearInterpolate(
   A=LinearInterpolate(
    A=Constant3Vector() {constant=<rgba 0.7305 0.5149 0.2346 1>}
    B=Constant3Vector() {constant=<rgba 0.4948 0.3488 0.1589 1>}
    Alpha=TextureSample() {texture=/Game/Global/Textures/Default/Tileable_Patterns/T_Blowing_Smoke_Pattern2_Tile_DISP}.R
   )
   B=Constant3Vector() {constant=<rgba 0.5972 0.3095 0.0908 1>}
   Alpha=TextureSample() {samplerType=SAMPLERTYPE_LinearColor, texture=/Game/Global/Textures/Default/Masks_and_Gradients/T_Cracks}.R
  )
  n1 = TextureSample(UVs=MaterialFunctionCall!(UVs=TextureCoordinate(), SubImages=Constant2Vector() {g=2, r=4}, Frame=ScalarParameter() {defaultValue=1, parameterName=SUBUV_Index}) {materialFunction=&/SubUV_Function}.UVs) {texture=/Game/Effects/Fort_Effects/Textures/Water/T_ChromeLiquid_4x2_D}
  Multiply(A=@n1.R)
  n2 = DecalLifetimeOpacity()
  Multiply(
   A=Multiply(
    A=@n1.RGB
    B=MaterialFunctionCall(Float2=@n2.Opacity) {materialFunction=&BreakOutFloat2Components}.R
   )
   B=DecalColor().A
  )
  Multiply(A=NamedRerouteUsage(), B=Constant3Vector() {constant=<rgba 0 0 3 0>})
  n3 = TextureCoordinate() {coordinateIndex=1}
  n4 = Add(
   A=Multiply(A=Subtract(A=@n3) {constB=0.5}, B=ScalarParameter() {defaultValue=0.5, parameterName=B_3}) {constB=0.5}
   B=AppendVector(A=ScalarParameter() {parameterName=X_Offset}, B=ScalarParameter() {defaultValue=0.5, parameterName=Y_Offset})
  )
  n5 = Constant()
  n6 = Constant()
  n7 = SphereMask(
   A=Add(A=TextureCoordinate() {coordinateIndex=1}, B=AppendVector(A=@n5, B=@n6)) {constB=0}
   B=ScalarParameter() {defaultValue=0.5, parameterName=CenterMaskU}
   Radius=Constant() {r=0.4}
   Hardness=ScalarParameter() {defaultValue=0.02, parameterName=HardnessMaskU}
  )
  n8 = TextureSample(UVs=TextureCoordinate() {uTiling=0.5, vTiling=3}) {texture=/BRCosmetics/Characters/Player/Male/Medium/Bodies/M_MED_NebulaBurden/Texture/T_NebulaBurden_OrbNoise}
  Multiply(
   A=TextureSample(UVs=Panner(Coordinate=@n4) {speedX=-1}) {texture=/BRCosmetics/Athena/Effects/Textures/T_InkySmokeTrail}.A
   B=NamedRerouteDeclaration(None=Multiply(A=@n7, B=@n8.R)) {name=MaskArea_1, nodeColor=<rgba 0.9688 1 0 1>}
  )
  Multiply(A=@n5)
  n9 = VectorParameter() {parameterName=Tile, defaultValue=<rgba 1 1 0 0>}
  n10 = TextureCoordinate()
  n11 = TextureSample(UVs=MaterialFunctionCall(Texture Scale=Constant() {r=1.4}) {materialFunction=&ScaleUVsByCenter}.UVs) {samplerType=SAMPLERTYPE_LinearColor, texture=/Game/Environments/Hexmap/HexClouds/T_RadialFlowmap01}
  n12 = DynamicParameter() {parameterIndex=1, defaultValue=<rgba 0 0 1 1>, paramNames [CenterWarp,MeshWPOBend,Param3,Param4]}
  n13 = Multiply(A=Subtract(A=ComponentMask(None=@n11.RGB) {g=true, r=true}) {constB=0.5}, B=@n12.CenterWarp)
  n14 = TextureSampleParameter2D(UVs=Multiply(A=AppendVector(A=@n9.R, B=@n9.R), B=Add(A=@n10, B=@n13))) {parameterName=Mask, texture=/Game/Effects/Fort_Effects/Textures/Water/T_water_splash_04}
  n15 = Multiply(
   A=ParticleColor().A
   B=MaterialFunctionCall(
    Fade Length=ScalarParameter() {parameterName=Camera Depth Fade Length}
    Fade Offset=ScalarParameter() {parameterName=Camera Depth Fade Offset}
   ) {materialFunction=&CameraDepthFade}.Result
  )
  n16 = Subtract(
   A=ChannelMaskParameter(None=@n14.RGB) {maskChannel=Blue, parameterName=ErodeMask, defaultValue=<rgba 0 0 1 0>}.R
   B=LinearInterpolate(Alpha=Saturate(None=@n15)) {constA=1, constB=0}
  )
  n17 = Constant() {r=0.2}
  Multiply(A=Saturate(None=Divide(A=@n16, B=@n17)), B=ChannelMaskParameter(None=@n14.RGBA) {parameterName=OpacityMask}.R)
  n18 = ScalarParameter() {defaultValue=1, parameterName=FadeIn}
  n19 = DecalLifetimeOpacity()
  n20 = Multiply(A=OneMinus(None=Multiply(A=@n18, B=@n19.Opacity)), B=ScalarParameter() {defaultValue=-5, parameterName=ErodeStrength})
  n21 = Multiply(A=TextureCoordinate()) {constB=0.5}
  n22 = TextureSample!(
   UVs=@n21
   Tex
   "DDX(UVs)"=DDX(Value=ComponentMask(None=@n21) {r=true})
   "DDY(UVs)"=DDY(Value=ComponentMask(None=@n21) {g=true})
   Apply View MipBias
  ) {
   mipValueMode=TMVM_Derivative
   samplerSource=SSM_Clamp_WorldGroupSettings
   samplerType=SAMPLERTYPE_LinearColor
   texture=/Game/Effects/Fort_Effects/Textures/Water/T_waterSplash
  }
  n23 = Add(A=Multiply(A=TextureCoordinate()) {constB=0.5}, B=Constant2Vector() {g=-0.5})
  n24 = DDX(Value=ComponentMask(None=@n23) {r=true})
  n25 = DDY(Value=ComponentMask(None=@n23) {g=true})
  n26 = Multiply(A=Add(
    A=@n22.G
    B=TextureSample(UVs=@n23, "DDX(UVs)"=@n24, "DDY(UVs)"=@n25) {mipValueMode=TMVM_Derivative, texture=/Game/Effects/Fort_Effects/Textures/Water/T_BoilBubble_D}.G
   ))
  NamedRerouteDeclaration(None=Multiply(
    A=MaterialFunctionCall!(Dynamic Param=@n20, Alpha=@n26) {materialFunction=/Game/Effects/Fort_Effects/Materials/_Functions/Erode_Alpha}.Result
    B=ScalarParameter() {defaultValue=0.25, parameterName=Opacity}
   )) {name=OPAAC, nodeColor=<rgba 0 0.4609 1 1>}
  n27 = TextureCoordinate() {coordinateIndex=1}
  n28 = AppendVector(A=ScalarParameter() {defaultValue=1, parameterName=N2_UMult}, B=ScalarParameter() {defaultValue=1, parameterName=N2_VMult})
  n29 = TextureCoordinate()
  n30 = AppendVector(A=ScalarParameter() {defaultValue=0.76, parameterName=N_UMult}, B=ScalarParameter() {defaultValue=0.2, parameterName=N_VMult})
  n31 = MaterialFunctionCall(
   BaseNormal=TextureSample!(UVs=Multiply(A=@n27, B=@n28), Tex, Apply View MipBias) {samplerType=SAMPLERTYPE_Normal, texture=/BRCosmetics/Animation/Game/MainPlayer/Emotes/TroutWrist/FX/Textures/T_FX_SphereNormals}.RGB
   AdditionalNormal=TextureSample(UVs=Multiply(A=@n29, B=@n30)) {samplerType=SAMPLERTYPE_LinearColor, texture=/Game/Environments/Asteria/Roads/Knights/Cobblestone/Textures/DirtPath_normal_SMRH}.RGB
  ) {materialFunction=&BlendAngleCorrectedNormals}
  n32 = ScalarParameter() {defaultValue=0.3, parameterName="Flatness (S)"}
  NamedRerouteDeclaration(None=NamedRerouteDeclaration(None=MaterialFunctionCall(Normal=@n31.Result, Flatness=@n32) {materialFunction=&FlattenNormal}.Result) {name=SphericalN, nodeColor=<rgba 0.2422 1 0 1>}) {name=Normals, nodeColor=<rgba 1 0.5859 0 1>}
  n33 = TextureCoordinate() {coordinateIndex=1}
  n34 = ScalarParameter() {defaultValue=0.6, parameterName=RadiusMaskMinU}
  n35 = ScalarParameter() {defaultValue=0.3, parameterName=RadiusMaskMaxU}
  n36 = Noise!(World Position=Multiply(A=WorldPosition().XYZ, B=ScalarParameter() {defaultValue=0.03, parameterName=PScale}), FilterWidth) {levels=2, outputMin=0}
  NamedRerouteDeclaration(None=SphereMask(
    A=ComponentMask(None=@n33) {g=true}
    B=ScalarParameter() {defaultValue=0.5, parameterName=CenterMaskU}
    Radius=LinearInterpolate(A=@n34, B=@n35, Alpha=@n36)
    Hardness=ScalarParameter() {defaultValue=0.02, parameterName=HardnessMaskU}
   )) {name=CenterRampMask, nodeColor=<rgba 1 0 0.4922 1>}
  n37 = TextureCoordinate() {coordinateIndex=1}
  n38 = AppendVector(A=Constant() {r=0.3}, B=Constant())
  NamedRerouteDeclaration(None=SphereMask(
    A=Add(A=@n37, B=@n38) {constB=0}
    B=ScalarParameter() {defaultValue=0.5, parameterName=CenterMaskU}
    Radius=Constant() {r=0.7}
    Hardness=ScalarParameter() {defaultValue=0.02, parameterName=HardnessMaskU}
   )) {name=MaskArea, nodeColor=<rgba 0.9688 1 0 1>}
  NamedRerouteUsage()
  NamedRerouteUsage()
  ScalarParameter() {defaultValue=0.5, parameterName=Radius}
  n39 = TextureSample() {texture=/Game/Missions/Secondary/BluGloSiphon/Textures/T_Bluglo_Splat_01}
  n40 = SphereMask(
   A=Multiply(
    A=TextureCoordinate()
    B=TextureSample() {texture=/BRCosmetics/Animation/Game/MainPlayer/Emotes/TroutWrist/FX/Textures/boost_MagicFire_BlobNoise_Glow_T}.R
   )
   B=Constant() {r=0.5}
   Radius=DynamicParameter() {paramNames [Radius, NoiseErodeStrength, NoiseScale, Param4]}.Radius
   Hardness=ScalarParameter() {defaultValue=0.5, parameterName=Hardness}
  )
  n41 = DynamicParameter() {paramNames [Radius, NoiseErodeStrength, NoiseScale, Param4]}
  n42 = Multiply(A=TextureCoordinate() {uTiling=2, vTiling=2}, B=@n41.NoiseScale)
  Subtract(
   A=Multiply(A=NamedRerouteUsage(), B=Multiply(A=@n39.R, B=@n40))
   B=Multiply(
    A=TextureSample(UVs=@n42) {samplerType=SAMPLERTYPE_Grayscale, texture=&LowResBlurredNoise}.R
    B=@n41.NoiseErodeStrength
   )
  )
  TextureSample!(UVs, Tex, "DDX(UVs)", "DDY(UVs)", Apply View MipBias) {mipValueMode=TMVM_Derivative, texture=&T_BoilBubble_D}
  TextureSample!(UVs, Tex, Apply View MipBias) {texture=/BRCosmetics/Animation/Game/MainPlayer/Emotes/TroutWrist/FX/Textures/tex_A}
 ]
}
