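# Niagara effect for the TroutWrist emote that throws banana-colored mesh bits and leaves fading decals around the character's feet.
# Foot-state flags such as grounded and max gate three bit emitters, which spawn at skeletal mesh bones chosen by spawn group, then fall, collide and shrink.
# Front-end rate spawning and gameplay distance spawning above speed thresholds are switched by FN_FrontEndFloatModifier.
# Decal001 places randomly chosen, fading decals at per-foot offsets, while the older Decal emitter is disabled.
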NiagaraSystem /BRCosmetics/Animation/Game/MainPlayer/Emotes/TroutWrist/FX/NS_Emote_TroutBananaBits
dependencies {
 usedDataInterfaces [
  &NiagaraDataInterfaceMeshRendererInfo
  &NiagaraDataInterfaceVectorCurve
  &NiagaraDataInterfaceCurve
  &NiagaraDataInterfaceArrayInt32
  &NiagaraDataInterfaceArrayFloat3
 ]
 usedDynamicInputs [&MakeFloatFromVector]
 usedModules [
  &self:SystemSpawnScript.SystemScriptSource.SystemScriptGraph.NiagaraNodeAssignment_1.SetVariables_#1
  &SystemState
  &SampleSkeletalMesh
  &self:SystemSpawnScript.SystemScriptSource.SystemScriptGraph.NiagaraNodeAssignment_0.SetVariables_#2
  &self:SystemSpawnScript.SystemScriptSource.SystemScriptGraph.NiagaraNodeAssignment_3.SetVariables_#3
  &self:SystemSpawnScript.SystemScriptSource.SystemScriptGraph.NiagaraNodeAssignment_4.SetVariables_#4
  &self:SystemSpawnScript.SystemScriptSource.SystemScriptGraph.NiagaraNodeAssignment_2.SetVariables_#5
 ]
}
emitters {
 BananaBits {
  emitterUpdateScript [
   &EmitterState(Life Cycle Mode=::NewEnumerator1, Loop Duration=1, UseLoopDelay=false)
   &SpawnRate(
    SpawnRate=&Lerp_Float(
     A=0
     B=&FN_FrontEndFloatModifier(Gameplay=0, Front End=50)
     Alpha=&MakeCustomFloatFromBool(Bool=$System.L_Foot_Max, True Float=1, False Float=0)
    )
    Use Spawn Probability=false
   )
   &SpawnRate(
    SpawnGroup=1
    SpawnRate=&Lerp_Float(
     A=0
     B=&FN_FrontEndFloatModifier(Gameplay=0, Front End=50)
     Alpha=&MakeCustomFloatFromBool(Bool=$System.R_Foot_Max, True Float=1, False Float=0)
    )
    Use Spawn Probability=false
   ) {module=SpawnRate002}
   &SpawnRate(SpawnGroup=1, SpawnRate=500, Use Spawn Probability=false) {enabled=false, module=SpawnRate001}
   &SpawnPerUnit(
    Module.Use Max Movement Threshold=true
    Module.Use Movement Tolerance=false
    Spawn Probability:NiagaraFloat=&FN_FrontEndFloatModifier(
     Gameplay=&MakeCustomFloatFromBool(
      Bool=&SetBoolByFloatComparison(A=&VectorLength(Vector=$Engine.Owner.Velocity), B=100)
      True Float=1
      False Float=0
     )
     Front End=0
    )
    Spawn Spacing:NiagaraFloat=&Lerp_Float(A=0, B=11, Alpha=&MakeCustomFloatFromBool(Bool=$System.L_Foot_Max, True Float=1, False Float=0))
    Use Spawn Probability=true
   )
   &SpawnPerUnit(
    Module.Use Max Movement Threshold=true
    Module.Use Movement Tolerance=false
    Spawn Group=1
    Spawn Probability:NiagaraFloat=&FN_FrontEndFloatModifier(
     Gameplay=&MakeCustomFloatFromBool(
      Bool=&SetBoolByFloatComparison(A=&VectorLength(Vector=$Engine.Owner.Velocity), B=100)
      True Float=1
      False Float=0
     )
     Front End=0
    )
    Spawn Spacing:NiagaraFloat=&Lerp_Float(A=0, B=11, Alpha=&MakeCustomFloatFromBool(Bool=$System.R_Foot_Max, True Float=1, False Float=0))
    Use Spawn Probability=true
   ) {module=SpawnPerUnit001}
  ]
  particleSpawnScript [
   &InitializeParticle(
    Color=<rgba 1 0.9574 0.2633 1>
    Lifetime=&MakeCustomFloatFromBool(Bool=$System.LFoot_Grounded, True Float=0.8, False Float=0)
    Mesh Renderer Array Visibility Mode=::NewEnumerator3
    Mesh Renderer Info=DataInterface() {propertyValues {MeshRenderer=&self:BananaBits_3.NiagaraMeshRendererProperties_2}}
    Mesh Scale Mode=::NewEnumerator3
    Mesh Uniform Scale=&RandomRangeFloat(Minimum=0.89, Maximum=1.2)
    Ribbon Facing Mode=::NewEnumerator1
    Ribbon Width=20
    Ribbon Width Mode=::NewEnumerator1
    UsePositionOffset=false
    Write Material Random=true
   )
   /SetVariables(Particles.VisibilityTag=&RandomRangeInt(Minimum=0, Maximum=2)) {enabled=false, isSetParameters=true, module=SetVariables_#6}
   &AddVelocity(Velocity=&RandomRangeVector(Minimum=<0 -50 50>, Maximum=<1 -350 200>))
   &SkeletalMeshLocation(
    Bone / Socket Index=&SpawnGroup()
    Bone Sampling Mode=::NewEnumerator4
    MaskSpawnGroup=false
    Skeletal Mesh=$System.Skeletal Mesh
   )
   &ShapeLocation(MaskSpawnGroup=false, Override Seed=false, Sphere Radius=20, UseOwnerScale=true)
   &InitialMeshOrientation(Use Rotation Vector=false)
  ]
  particleUpdateScript [
   &ParticleState()
   &ScaleMeshSize(Scale Factor=&VectorFromCurve(
     VectorCurve=DataInterface() {propertyValues {XCurve {keys [<at 0 1>,<at 1 0>]}, YCurve {keys [<at 0 1>,<at 1 0>]}, ZCurve {keys [<at 0 1>,<at 1 0>]}}}
     Scale Curve=&RandomRangeVector(Minimum=<0.7 0.7 0.7>, Maximum=<1 1 1>)
    ))
   &GravityForce() {module=GravityForce001}
   &Collision(
    Age Colliding Particles=false
    Friction=0.8
    KillOnCollision=false
    Module.Control Rest State Via Surface Normal=false
    Module.EnableMaxCollisionCount=false
    Module.Randomize Collision Normal Vector=false
    Restitution=0.1
   )
   &SolveForcesAndVelocity(Clamp Velocity=false, Limit Acceleration=false)
   &UpdateMeshOrientation(
    Mesh Radius Calculation Method:ENiagaraCalculateRadiusOptions=::NewEnumerator0
    Mesh Renderer=DataInterface() {propertyValues {MeshRenderer=&self:BananaBits_3.NiagaraMeshRendererProperties_2}}
    Mesh Renderer Index:NiagaraInt32=&RandomRangeInt(Minimum=0, Maximum=2)
    Orientation Method=::NewEnumerator6
    Rotation Rate=0.5
    Surface Normal:Vector3f=$Particles.Collision.SafeCombinedCollisionNormal
   )
   &DynamicMaterialParameters(
    Index 0 Param 1=90
    Index 0 Param 2=0.1
    Index 0 Param 3=1.0911
    Module.Index0Param0Write=true
    Module.Index0Param1Write=true
    Module.Index0Param2Write=false
    Module.Index0Param3Write=false
   )
   /SetVariables(Particles.MeshOrientation=&LerpQuaternion(
     Quaternion A:Quat4f=<0 0 1 1>
     Quaternion B:Quat4f=$Particles.MeshOrientation
     Lerp Factor:NiagaraFloat=&FloatFromCurve(
      FloatCurve=DataInterface() {propertyValues {Curve {keys [{arriveTangent=-3.7053, interpMode=RCIM_Cubic, leaveTangent=-3.7053, tangentMode=RCTM_User, time=0, value=2}, <at 0.65 0 RCIM_Cubic>]}}}
      Scale Curve=2
     )
     Enable Spherical Linear Interpolation:NiagaraBool=true
    )) {isSetParameters=true}
  ]
  rendererClasses [&NiagaraMeshRendererProperties]
  renderers [{rendererClass=&NiagaraMeshRendererProperties, rendererIndex=0}]
 }
 BananaBits001 {
  emitterUpdateScript [
   &EmitterState(Life Cycle Mode=::NewEnumerator1, Loop Duration=1, UseLoopDelay=false)
   &SpawnRate(
    SpawnRate=&Lerp_Float(A=0, B=200, Alpha=&MakeCustomFloatFromBool(Bool=$User.bIsFrontEnd, True Float=0, False Float=1))
    Use Spawn Probability=false
   ) {enabled=false}
   &SpawnRate(
    SpawnGroup=1
    SpawnRate=&FN_FrontEndFloatModifier(Gameplay=0, Front End=80)
    Use Spawn Probability=false
   ) {module=SpawnRate001}
   &SpawnPerUnit(
    Module.Use Max Movement Threshold=true
    Module.Use Movement Tolerance=false
    Spawn Probability:NiagaraFloat=&FN_FrontEndFloatModifier(
     Gameplay=&MakeCustomFloatFromBool(
      Bool=&SetBoolByFloatComparison(A=&VectorLength(Vector=$Engine.Owner.Velocity), B=100)
      True Float=1
      False Float=0
     )
     Front End=0
    )
    Spawn Spacing:NiagaraFloat=&Lerp_Float(A=0, B=7, Alpha=&MakeCustomFloatFromBool(Bool=$System.LFoot_Grounded, True Float=1, False Float=0))
    Use Spawn Probability=true
   )
   &SpawnPerUnit(
    Module.Use Max Movement Threshold=true
    Module.Use Movement Tolerance=false
    Spawn Group=1
    Spawn Probability:NiagaraFloat=&FN_FrontEndFloatModifier(
     Gameplay=&MakeCustomFloatFromBool(
      Bool=&SetBoolByFloatComparison(A=&VectorLength(Vector=$Engine.Owner.Velocity), B=100)
      True Float=1
      False Float=0
     )
     Front End=0
    )
    Spawn Spacing:NiagaraFloat=&Lerp_Float(A=0, B=9, Alpha=&MakeCustomFloatFromBool(Bool=$System.RFoot_Grounded, True Float=1, False Float=0))
    Use Spawn Probability=true
   ) {module=SpawnPerUnit001}
  ]
  particleSpawnScript [
   &InitializeParticle(
    Color=<rgba 1 0.9574 0.2633 1>
    Lifetime=0.8
    Mesh Renderer Array Visibility Mode=::NewEnumerator3
    Mesh Renderer Info=DataInterface() {propertyValues {MeshRenderer=&self:BananaBits001_8.NiagaraMeshRendererProperties_2}}
    Mesh Scale Mode=::NewEnumerator3
    Mesh Uniform Scale=&RandomRangeFloat(Minimum=0.3, Maximum=0.5)
    Ribbon Facing Mode=::NewEnumerator1
    Ribbon Width=20
    Ribbon Width Mode=::NewEnumerator1
    UsePositionOffset=false
    Write Material Random=true
   )
   /SetVariables(Particles.VisibilityTag=&RandomRangeInt(Minimum=0, Maximum=2)) {enabled=false, isSetParameters=true, module=SetVariables_#6}
   &AddVelocity(Velocity=&RandomRangeVector(Minimum=<0 -50 50>, Maximum=<1 -350 200>))
   &SkeletalMeshLocation(
    Bone / Socket Index=&SpawnGroup()
    Bone Sampling Mode=::NewEnumerator4
    MaskSpawnGroup=false
    Skeletal Mesh=$System.Skeletal Mesh
   )
   &ShapeLocation(MaskSpawnGroup=false, Override Seed=false, Sphere Radius=20, UseOwnerScale=true)
   &InitialMeshOrientation(Use Rotation Vector=false)
  ]
  particleUpdateScript [
   &ParticleState()
   &ScaleMeshSize(Scale Factor=&VectorFromCurve(
     VectorCurve=DataInterface() {propertyValues {XCurve {keys [<at 0 1>,<at 1 0>]}, YCurve {keys [<at 0 1>,<at 1 0>]}, ZCurve {keys [<at 0 1>,<at 1 0>]}}}
     Scale Curve=&RandomRangeVector(Minimum=<0.7 0.7 0.7>, Maximum=<1 1 1>)
    ))
   &GravityForce() {module=GravityForce001}
   &Collision(
    Age Colliding Particles=false
    Friction=0.8
    KillOnCollision=false
    Module.Control Rest State Via Surface Normal=false
    Module.EnableMaxCollisionCount=false
    Module.Randomize Collision Normal Vector=false
    Restitution=0.1
   )
   &SolveForcesAndVelocity(Clamp Velocity=false, Limit Acceleration=false)
   &UpdateMeshOrientation(
    Mesh Radius Calculation Method:ENiagaraCalculateRadiusOptions=::NewEnumerator0
    Mesh Renderer=DataInterface() {propertyValues {MeshRenderer=&self:BananaBits001_8.NiagaraMeshRendererProperties_2}}
    Mesh Renderer Index:NiagaraInt32=&RandomRangeInt(Minimum=0, Maximum=2)
    Orientation Method=::NewEnumerator6
    Rotation Rate=0.5
    Surface Normal:Vector3f=$Particles.Collision.SafeCombinedCollisionNormal
   )
   &DynamicMaterialParameters(
    Index 0 Param 1=90
    Index 0 Param 2=0.1
    Index 0 Param 3=1.0911
    Module.Index0Param0Write=true
    Module.Index0Param1Write=true
    Module.Index0Param2Write=false
    Module.Index0Param3Write=false
   )
   /SetVariables(Particles.MeshOrientation=&LerpQuaternion(
     Quaternion A:Quat4f=<0 0 1 1>
     Quaternion B:Quat4f=$Particles.MeshOrientation
     Lerp Factor:NiagaraFloat=&FloatFromCurve(
      FloatCurve=DataInterface() {propertyValues {Curve {keys [{arriveTangent=-3.7053, interpMode=RCIM_Cubic, leaveTangent=-3.7053, tangentMode=RCTM_User, time=0, value=2}, <at 0.65 0 RCIM_Cubic>]}}}
      Scale Curve=2
     )
     Enable Spherical Linear Interpolation:NiagaraBool=true
    )) {isSetParameters=true}
  ]
  rendererClasses [&NiagaraMeshRendererProperties]
  renderers [{rendererClass=&NiagaraMeshRendererProperties, rendererIndex=0}]
 }
 Banana_lifted {
  emitterUpdateScript [
   &EmitterState(Life Cycle Mode=::NewEnumerator1, Loop Duration=1, UseLoopDelay=false)
   &SpawnRate(
    SpawnRate=&Lerp_Float(A=0, B=7, Alpha=&MakeCustomFloatFromBool(Bool=$System.LFoot_Grounded, True Float=0, False Float=1))
    Use Spawn Probability=false
   )
   &SpawnRate(
    SpawnGroup=1
    SpawnRate=&Lerp_Float(A=0, B=5, Alpha=&MakeCustomFloatFromBool(Bool=$System.RFoot_Grounded, True Float=0, False Float=1))
    Use Spawn Probability=false
   ) {module=SpawnRate001}
   &SpawnPerUnit(
    Module.Use Max Movement Threshold=true
    Module.Use Movement Tolerance=false
    Spawn Probability:NiagaraFloat=&FN_FrontEndFloatModifier(
     Gameplay=&MakeCustomFloatFromBool(
      Bool=&SetBoolByFloatComparison(A=&VectorLength(Vector=$Engine.Owner.Velocity), B=150)
      True Float=0
      False Float=1
     )
     Front End=0
    )
    Spawn Spacing:NiagaraFloat=&Lerp_Float(A=0, B=40, Alpha=&MakeCustomFloatFromBool(Bool=$System.LFoot_Grounded, True Float=0, False Float=1))
    Use Spawn Probability=true
   ) {enabled=false}
   &SpawnPerUnit(
    Module.Use Max Movement Threshold=true
    Module.Use Movement Tolerance=false
    Spawn Probability:NiagaraFloat=&FN_FrontEndFloatModifier(
     Gameplay=&MakeCustomFloatFromBool(
      Bool=&SetBoolByFloatComparison(A=&VectorLength(Vector=$Engine.Owner.Velocity), B=150)
      True Float=1
      False Float=0
     )
     Front End=0
    )
    Spawn Spacing:NiagaraFloat=&Lerp_Float(A=0, B=40, Alpha=&MakeCustomFloatFromBool(Bool=$System.RFoot_Grounded, True Float=1, False Float=0))
    Use Spawn Probability=true
   ) {enabled=false, module=SpawnPerUnit001}
  ]
  particleSpawnScript [
   &InitializeParticle(
    Color=<rgba 1 0.9574 0.2633 1>
    Lifetime=0.8
    Mesh Renderer Array Visibility Mode=::NewEnumerator3
    Mesh Renderer Info=DataInterface() {propertyValues {MeshRenderer=&self:Banana_lifted_2.NiagaraMeshRendererProperties_2}}
    Mesh Scale Mode=::NewEnumerator3
    Mesh Uniform Scale=&RandomRangeFloat(Minimum=0.3, Maximum=0.8)
    Ribbon Facing Mode=::NewEnumerator1
    Ribbon Width=20
    Ribbon Width Mode=::NewEnumerator1
    UsePositionOffset=false
    Write Material Random=true
   )
   /SetVariables(Particles.VisibilityTag=&RandomRangeInt(Minimum=0, Maximum=2)) {enabled=false, isSetParameters=true, module=SetVariables_#6}
   &AddVelocity(Velocity=&RandomRangeVector(Minimum=<0 -50 50>, Maximum=<1 -350 200>))
   &SkeletalMeshLocation(
    Bone / Socket Index=&SpawnGroup()
    Bone Sampling Mode=::NewEnumerator4
    MaskSpawnGroup=false
    Skeletal Mesh=$System.Skeletal Mesh
   )
   &ShapeLocation(MaskSpawnGroup=false, Override Seed=false, Sphere Radius=20, UseOwnerScale=true)
   &InitialMeshOrientation(Use Rotation Vector=false)
  ]
  particleUpdateScript [
   &ParticleState()
   &ScaleMeshSize(Scale Factor=&VectorFromCurve(VectorCurve=DataInterface() {propertyValues {XCurve {keys [<at 0 1>,<at 1 0>]}, YCurve {keys [<at 0 1>,<at 1 0>]}, ZCurve {keys [<at 0 1>,<at 1 0>]}}}))
   &GravityForce() {module=GravityForce001}
   &Collision(
    Age Colliding Particles=false
    Friction=0.8
    KillOnCollision=false
    Module.Control Rest State Via Surface Normal=false
    Module.EnableMaxCollisionCount=false
    Module.Randomize Collision Normal Vector=false
    Restitution=0.1
   )
   &SolveForcesAndVelocity(Clamp Velocity=false, Limit Acceleration=false)
   &UpdateMeshOrientation(
    Mesh Radius Calculation Method:ENiagaraCalculateRadiusOptions=::NewEnumerator0
    Mesh Renderer=DataInterface() {propertyValues {MeshRenderer=&self:Banana_lifted_2.NiagaraMeshRendererProperties_2}}
    Mesh Renderer Index:NiagaraInt32=&RandomRangeInt(Minimum=0, Maximum=2)
    Orientation Method=::NewEnumerator6
    Rotation Rate=0.5
    Surface Normal:Vector3f=$Particles.Collision.SafeCombinedCollisionNormal
   )
   &DynamicMaterialParameters(
    Index 0 Param 1=90
    Index 0 Param 2=0.1
    Index 0 Param 3=1.0911
    Module.Index0Param0Write=true
    Module.Index0Param1Write=true
    Module.Index0Param2Write=false
    Module.Index0Param3Write=false
   )
   /SetVariables(Particles.MeshOrientation=&LerpQuaternion(
     Quaternion A:Quat4f=<0 0 1 1>
     Quaternion B:Quat4f=$Particles.MeshOrientation
     Lerp Factor:NiagaraFloat=&FloatFromCurve(
      FloatCurve=DataInterface() {propertyValues {Curve {keys [{arriveTangent=-3.7053, interpMode=RCIM_Cubic, leaveTangent=-3.7053, tangentMode=RCTM_User, time=0, value=2}, <at 0.65 0 RCIM_Cubic>]}}}
      Scale Curve=2
     )
     Enable Spherical Linear Interpolation:NiagaraBool=true
    )) {isSetParameters=true}
  ]
  rendererClasses [&NiagaraMeshRendererProperties]
  renderers [{rendererClass=&NiagaraMeshRendererProperties, rendererIndex=0}]
 }
 Decal {
  bEnabled=false
  emitterUpdateScript [
   &EmitterState(Inactive Response=::NewEnumerator1, Life Cycle Mode=::NewEnumerator1, Loop Duration=1, UseLoopDelay=false)
   &SpawnRate(SpawnRate=20, Use Spawn Probability=false) {enabled=false}
   &SpawnPerUnit(
    Module.Use Max Movement Threshold=true
    Module.Use Movement Tolerance=false
    Spawn Probability:NiagaraFloat=&FN_FrontEndFloatModifier(
     Gameplay=&MakeCustomFloatFromBool(
      Bool=&SetBoolByFloatComparison(A=&VectorLength(Vector=$Engine.Owner.Velocity), B=150)
      True Float=1
      False Float=0
     )
     Front End=0
    )
    Spawn Spacing:NiagaraFloat=&Lerp_Float(A=0, B=5, Alpha=&MakeCustomFloatFromBool(Bool=$System.LFoot_Grounded, True Float=0, False Float=1))
    Use Spawn Probability=true
   )
   &SpawnPerUnit(
    Module.Use Max Movement Threshold=true
    Module.Use Movement Tolerance=false
    Spawn Probability:NiagaraFloat=&FN_FrontEndFloatModifier(
     Gameplay=&MakeCustomFloatFromBool(
      Bool=&SetBoolByFloatComparison(A=&VectorLength(Vector=$Engine.Owner.Velocity), B=150)
      True Float=1
      False Float=0
     )
     Front End=0
    )
    Spawn Spacing:NiagaraFloat=&Lerp_Float(A=0, B=5, Alpha=&MakeCustomFloatFromBool(Bool=$System.RFoot_Grounded, True Float=0, False Float=1))
    Use Spawn Probability=true
   ) {module=SpawnPerUnit001}
  ]
  particleSpawnScript [
   &InitializeParticle(
    Lifetime=2
    Mesh Scale=&Multiply_VectorByFloat(Vector=<20 50 50>, Float=&RandomRangeFloat(Minimum=0.2, Maximum=1))
    Mesh Scale Mode=::NewEnumerator4
    Sprite Size Mode=::NewEnumerator3
    Uniform Sprite Size=-2.5
    UsePositionOffset=false
    Write Material Random=true
   )
   &PartitionParticles(Number Of Partitions=3)
   &InitialMeshOrientation(
    Mesh Axis To Orient=<0 0 1>
    Mesh Orientation Mode=::NewEnumerator2
    Rotation=&RandomRangeVector(Minimum=<0 0 0>, Maximum=<1 0 0>)
    Use Rotation Vector=true
   )
   &SkeletalMeshLocation(
    Bone / Socket Index=&SelectIntFromArray(Direct Array Index=$Particles.PartitionParticles.Partition, Int Selection Array=DataInterface() {propertyValues {IntData [0,1]}})
    Bone Sampling Mode=::NewEnumerator4
    MaskSpawnGroup=false
    Orientation Sampling=::NewEnumerator1
    Sampled Position Offset=<0 10 -10>
    Sampled Position Offset Coordinate Space:ENiagaraSkelMeshTransforms=::NewEnumerator1
    Skeletal Mesh=$System.Skeletal Mesh
   )
   /SetVariables(Particles.VisibilityTag=$Particles.PartitionParticles.Partition) {isSetParameters=true}
  ]
  particleUpdateScript [
   &ParticleState(Kill Particles When Lifetime Has Elapsed=false)
   &ScaleColor(ScaleA=true, ScaleRGB=true) {enabled=false}
   /SetVariables(Particles.DecalSize:Vector3f=$Particles.Scale) {isSetParameters=true, module=SetVariables_#7}
   /SetVariables(Particles.DecalFade:NiagaraFloat=&FloatFromCurve(FloatCurve=DataInterface() {propertyValues {Curve {keys [<at 0.1 0.9847 RCIM_Cubic>, {arriveTangent=-2.8714, interpMode=RCIM_Cubic, leaveTangent=-2.8714, tangentMode=RCTM_User, time=1, value=0}]}}})) {isSetParameters=true, module=SetVariables_#8}
   /SetVariables(Particles.DecalOrientation:Quat4f=$Particles.MeshOrientation) {isSetParameters=true, module=SetVariables_#9}
  ]
  rendererClasses [&NiagaraDecalRendererProperties]
  renderers [
   {rendererClass=&NiagaraDecalRendererProperties, rendererIndex=0}
   {rendererClass=&NiagaraDecalRendererProperties, rendererIndex=1}
   {rendererClass=&NiagaraDecalRendererProperties, rendererIndex=2}
  ]
 }
 Decal001 {
  emitterUpdateScript [
   &EmitterState(Inactive Response=::NewEnumerator1, Life Cycle Mode=::NewEnumerator1, Loop Duration=1, UseLoopDelay=false)
   &SpawnRate(SpawnRate=10, Use Spawn Probability=false) {enabled=false}
   &SpawnPerUnit(
    Module.Use Max Movement Threshold=true
    Module.Use Movement Tolerance=false
    Spawn Probability:NiagaraFloat=&FN_FrontEndFloatModifier(
     Gameplay=&MakeCustomFloatFromBool(
      Bool=&SetBoolByFloatComparison(A=&VectorLength(Vector=$Engine.Owner.Velocity), B=150)
      True Float=1
      False Float=0
     )
     Front End=0
    )
    Spawn Spacing:NiagaraFloat=&Lerp_Float(A=0, B=40, Alpha=&MakeCustomFloatFromBool(Bool=$System.LFoot_Grounded, True Float=0, False Float=1))
    Use Spawn Probability=true
   )
   &SpawnPerUnit(
    Module.Use Max Movement Threshold=true
    Module.Use Movement Tolerance=false
    Spawn Group=1
    Spawn Probability:NiagaraFloat=&FN_FrontEndFloatModifier(
     Gameplay=&MakeCustomFloatFromBool(
      Bool=&SetBoolByFloatComparison(A=&VectorLength(Vector=$Engine.Owner.Velocity), B=150)
      True Float=1
      False Float=0
     )
     Front End=0
    )
    Spawn Spacing:NiagaraFloat=&Lerp_Float(A=0, B=40, Alpha=&MakeCustomFloatFromBool(Bool=$System.RFoot_Grounded, True Float=0, False Float=1))
    Use Spawn Probability=true
   ) {module=SpawnPerUnit001}
  ]
  particleSpawnScript [
   &InitializeParticle(
    Lifetime=2
    Mesh Scale=&Multiply_VectorByFloat(Vector=<20 50 20>, Float=&RandomRangeFloat(Minimum=1.2, Maximum=1.5))
    Mesh Scale Mode=::NewEnumerator4
    Sprite Size Mode=::NewEnumerator3
    Uniform Sprite Size=2
    UsePositionOffset=false
    Write Material Random=true
   )
   &PartitionParticles(Number Of Partitions=4)
   &SkeletalMeshLocation(
    Bone / Socket Index=2
    Bone Sampling Mode=::NewEnumerator4
    MaskSpawnGroup=false
    Orientation Sampling=::NewEnumerator1
    Sampled Position Offset=&SelectVectorFromArray(
     Direct Array Index=&SpawnGroup()
     Vector Selection Array=DataInterface() {propertyValues {FloatData [<27 10 0>, <-30 10 0>]}}
    )
    Sampled Position Offset Coordinate Space:ENiagaraSkelMeshTransforms=::NewEnumerator2
    Skeletal Mesh=$System.Skeletal Mesh
   )
   &InitialMeshOrientation(Mesh Orientation Mode=::NewEnumerator1, Rotation=<0.5 0 1.4>, Use Rotation Vector=true)
   /SetVariables(Particles.DecalOrientation:Quat4f=$Particles.MeshOrientation) {isSetParameters=true, module=SetVariables_#9}
   /SetVariables(Particles.VisibilityTag=&RandomRangeInt(Minimum=0, Maximum=3)) {isSetParameters=true}
  ]
  particleUpdateScript [
   &ParticleState()
   &ScaleMeshSize(Scale Factor=&VectorFromCurve(VectorCurve=DataInterface() {propertyValues {
       XCurve {keys [
         {interpMode=RCIM_Cubic, leaveTangent=7.5145, tangentMode=RCTM_Break, time=0, value=0}
         {interpMode=RCIM_Cubic, tangentMode=RCTM_Break, time=0.2, value=1}
        ]}
       YCurve {keys [
         {interpMode=RCIM_Cubic, leaveTangent=7.5145, tangentMode=RCTM_Break, time=0, value=0}
         {interpMode=RCIM_Cubic, tangentMode=RCTM_Break, time=0.2, value=1}
        ]}
       ZCurve {keys [
         {interpMode=RCIM_Cubic, leaveTangent=7.5145, tangentMode=RCTM_Break, time=0, value=0}
         {interpMode=RCIM_Cubic, tangentMode=RCTM_Break, time=0.2, value=1}
        ]}
      }}))
   /SetVariables(Particles.DecalSize:Vector3f=$Particles.Scale) {isSetParameters=true, module=SetVariables_#7}
   /SetVariables(Particles.DecalFade:NiagaraFloat=&FloatFromCurve(FloatCurve=DataInterface() {propertyValues {Curve {keys [<at 0.1 0.9847 RCIM_Cubic>, {arriveTangent=-2.8714, interpMode=RCIM_Cubic, leaveTangent=-2.8714, tangentMode=RCTM_User, time=1, value=0}]}}})) {isSetParameters=true, module=SetVariables_#8}
  ]
  rendererClasses [&NiagaraDecalRendererProperties, &NiagaraMeshRendererProperties]
  renderers [
   {rendererClass=&NiagaraDecalRendererProperties, rendererIndex=0}
   {rendererClass=&NiagaraMeshRendererProperties, rendererIndex=1}
   {rendererClass=&NiagaraDecalRendererProperties, rendererIndex=2}
   {rendererClass=&NiagaraDecalRendererProperties, rendererIndex=3}
   {rendererClass=&NiagaraDecalRendererProperties, rendererIndex=4}
   {rendererClass=&NiagaraDecalRendererProperties, rendererIndex=5}
  ]
 }
}
summary {
 systemName=NS_Emote_TroutBananaBits
 emitters [
  {emitterName=BananaBits, rendererClasses [&NiagaraMeshRendererProperties]}
  {emitterName=Banana_lifted, rendererClasses [&NiagaraMeshRendererProperties]}
  {bEnabled=false, emitterName=Decal, rendererClasses [&NiagaraDecalRendererProperties]}
  {emitterName=BananaBits001, rendererClasses [&NiagaraMeshRendererProperties]}
  {emitterName=Decal001, rendererClasses [&NiagaraDecalRendererProperties, &NiagaraMeshRendererProperties]}
 ]
 userVariables [
  User.bIsFrontEnd {type=NiagaraBool}
  User.bIsFrontEndPreview {type=NiagaraBool}
  User.Color {type=LinearColor, description="", defaultValue {struct=LinearColor, value=<rgba 0.624 0.4678 0.227 1>}}
  User.DitherAmt {type=NiagaraFloat, defaultValue=0, description=""}
  User.Lifetime {type=NiagaraFloat, defaultValue=0, description=""}
  User.Skeletal Mesh {
   type=NiagaraDataInterfaceSkeletalMesh
   defaultValue {
    struct=NiagaraExt_VariableValue_DataInterface
    value {dataInterface=&self:NiagaraDataInterfaceSkeletalMesh_0}
   }
  }
  User.WorldPosScale {type=NiagaraFloat, defaultValue=0.1, description=""}
 ]
}
