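# Niagara foot trail effect for the TroutWrist emote, drawing tan mesh particles from one emitter.
# In gameplay, two distance-based spawners follow system left and right foot-grounded flags and an owner speed threshold; the front end uses a steady spawn rate instead.
# Particles are partitioned between two skeletal mesh bones or sockets and repositioned every update.
# A per-particle visibility tag, driven by those grounded flags, switches between two mesh renderers or neither.
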NiagaraSystem /BRCosmetics/Animation/Game/MainPlayer/Emotes/TroutWrist/FX/NS_Emote_TroutTrail
dependencies {
 usedDataInterfaces [&NiagaraDataInterfaceArrayInt32, &NiagaraDataInterfaceSkeletalMesh]
 usedDynamicInputs [&MakeFloatFromVector]
 usedModules [
  &self:SystemSpawnScript.SystemScriptSource.SystemScriptGraph.NiagaraNodeAssignment_1.SetVariables_#1
  &SystemState
  &SampleSkeletalMesh
  &self:SystemSpawnScript.SystemScriptSource.SystemScriptGraph.NiagaraNodeAssignment_0.SetVariables_#2
  &self:SystemSpawnScript.SystemScriptSource.SystemScriptGraph.NiagaraNodeAssignment_2.SetVariables_#3
 ]
}
emitters {BananaTrail002 {
  emitterUpdateScript [
   &EmitterState(Life Cycle Mode=::NewEnumerator1, Loop Duration=1, UseLoopDelay=false)
   &SpawnRate(SpawnRate=&FN_FrontEndFloatModifier(Gameplay=0, Front End=20), Use Spawn Probability=false) {module=SpawnRate002}
   &SpawnPerUnit(
    Module.Use Max Movement Threshold=true
    Module.Use Movement Tolerance=false
    Spawn Probability:NiagaraFloat=&FN_FrontEndFloatModifier(
     Gameplay=&MakeCustomFloatFromBool(
      Bool=&SetBoolByFloatComparison(A=&VectorLength(Vector=$Engine.Owner.Velocity), B=150)
      True Float=1
      False Float=0
     )
     Front End=0
    )
    Spawn Spacing:NiagaraFloat=&Lerp_Float(A=0, B=20, Alpha=&MakeCustomFloatFromBool(Bool=$System.LFoot_Grounded, True Float=1, False Float=0))
    Use Spawn Probability=true
   )
   &SpawnPerUnit(
    Module.Use Max Movement Threshold=true
    Module.Use Movement Tolerance=false
    Spawn Probability:NiagaraFloat=&FN_FrontEndFloatModifier(
     Gameplay=&MakeCustomFloatFromBool(
      Bool=&SetBoolByFloatComparison(A=&VectorLength(Vector=$Engine.Owner.Velocity), B=150)
      True Float=1
      False Float=0
     )
     Front End=0
    )
    Spawn Spacing:NiagaraFloat=&Lerp_Float(A=0, B=20, Alpha=&MakeCustomFloatFromBool(Bool=$System.RFoot_Grounded, True Float=1, False Float=0))
    Use Spawn Probability=true
   ) {module=SpawnPerUnit001}
  ]
  particleSpawnScript [
   &InitializeParticle(
    Color Maximum=<rgba 0.6939 0.491 0.2159 1>
    Color Minimum=<rgba 0.6654 0.4564 0.1946 1>
    Color Mode=::NewEnumerator2
    Mesh Scale=&MakeVector(
     X=40
     Y=&Lerp_Float(A=8, B=$User.Y_Mesh Scale, Alpha=&FN_FrontEndFloatModifier(Gameplay=0, Front End=1))
     Z=10
    )
    Mesh Scale Mode=::NewEnumerator4
    Ribbon Facing Mode=::NewEnumerator1
    Ribbon Width=20
    Ribbon Width Mode=::NewEnumerator1
    UsePositionOffset=false
    Write Material Random=true
   )
   &PartitionParticles()
  ]
  particleUpdateScript [
   &ParticleState()
   &SkeletalMeshLocation(
    Bone / Socket Index=&SelectIntFromArray(Direct Array Index=$Particles.PartitionParticles.Partition, Int Selection Array=DataInterface() {propertyValues {IntData [0,1]}})
    Bone Sampling Mode=::NewEnumerator4
    MaskSpawnGroup=false
    Orientation Sampling=::NewEnumerator1
    Sampled Position Offset=<0 10 -10>
    Sampled Position Offset Coordinate Space:ENiagaraSkelMeshTransforms=::NewEnumerator1
    Skeletal Mesh=$System.Skeletal Mesh
   )
   /SetVariables(Particles.VisibilityTag=&MakeIntFromFloat(
     Float=&Lerp_Float(
      A=&MakeCustomFloatFromBool(Bool=$System.LFoot_Grounded, True Float=0, False Float=2)
      B=&MakeCustomFloatFromBool(Bool=$System.RFoot_Grounded, True Float=1, False Float=2)
      Alpha=&MakeFloatFromInt(Int=$Particles.PartitionParticles.Partition)
     )
     Conversion Mode=::NewEnumerator1
    )) {isSetParameters=true}
   &SolveForcesAndVelocity(Clamp Velocity=false, Limit Acceleration=false)
   &DynamicMaterialParameters(
    Index 0 Param 1=90
    Index 0 Param 2=0.03
    Index 0 Param 3=1.0911
    Module.Index0Param0Write=true
    Module.Index0Param1Write=true
    Module.Index0Param2Write=false
    Module.Index0Param3Write=false
   )
  ]
  rendererClasses [&NiagaraMeshRendererProperties]
  renderers [
   {rendererClass=&NiagaraMeshRendererProperties, rendererIndex=0}
   {rendererClass=&NiagaraMeshRendererProperties, rendererIndex=1}
  ]
 }}
summary {
 systemName=NS_Emote_TroutTrail
 emitters [{emitterName=BananaTrail002, rendererClasses [&NiagaraMeshRendererProperties]}]
 userVariables [
  User.bIsFrontEnd {type=NiagaraBool}
  User.bIsFrontEndPreview {type=NiagaraBool}
  User.Color {type=LinearColor, description="", defaultValue {struct=LinearColor, value=<rgba 0.624 0.4678 0.227 1>}}
  User.DitherAmt {type=NiagaraFloat, defaultValue=0, description=""}
  User.Lifetime {type=NiagaraFloat, defaultValue=0, description=""}
  User.WorldPosScale {type=NiagaraFloat, defaultValue=0.1, description=""}
  User.Y_Mesh Scale {type=NiagaraFloat, defaultValue=13, description=""}
 ]
}
